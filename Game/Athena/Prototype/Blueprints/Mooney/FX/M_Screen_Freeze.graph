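# Translucent unlit screen-freeze material that frosts and whites out the view by world-space proximity to an Ice Center parameter.
# Three tiled T_LightFrost channels are summed, then shaped by a centered UV mask and sphere masks scaled from one Radius.
# CountdownStarted scales down the frost and removes the inner whiteout.
# Emissive adds whiteout-driven snow color to MF_SharedWinter ice color; its frost-masked lerp has identical ice-color inputs.
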Material /Game/Athena/Prototype/Blueprints/Mooney/FX/M_Screen_Freeze
graph {
 properties {
  bDisableDepthTest=true
  bEnableResponsiveAA=true
  bTangentSpaceNormal=false
  blendMode=BLEND_Translucent
  ditherOpacityMask=true
  pixelDepthOffsetMode=PDOM_Legacy
  shadingModel=MSM_Unlit
  twoSided=true
  usedShadingModels=Unlit
 }
 nodes [
  n1 = TextureCoordinate()
  n2 = AppendVector(A=ScalarParameter() {defaultValue=12, parameterName=Frost_U_Tile}, B=ScalarParameter() {defaultValue=12, parameterName=Frost_V_Tile})
  n3 = Multiply(
   A=TextureCoordinate()
   B=AppendVector(A=ScalarParameter() {defaultValue=20, parameterName=Frost_U_Tile2}, B=ScalarParameter() {defaultValue=20, parameterName=Frost_V_Tile2})
  )
  n4 = AppendVector(A=ScalarParameter() {defaultValue=0.06, parameterName=Frost_U_Offset}, B=ScalarParameter() {defaultValue=0.05, parameterName=Frost_V_Offset})
  n5 = Add(
   A=TextureSample(UVs=Multiply(A=@n1, B=@n2)) {texture=/Game/Effects/Fort_Effects/Textures/Snowflake/T_LightFrost}.R
   B=TextureSample(UVs=Add(A=@n3, B=@n4)) {texture=&T_LightFrost}.G
  )
  n6 = Multiply(
   A=TextureCoordinate()
   B=AppendVector(A=ScalarParameter() {defaultValue=50, parameterName=Frost_U_Tile3}, B=ScalarParameter() {defaultValue=50, parameterName=Frost_V_Tile3})
  )
  n7 = AppendVector(A=ScalarParameter() {defaultValue=0.3, parameterName=Frost_U_Offset2, sliderMin=0.06}, B=ScalarParameter() {parameterName=Frost_V_Offset2})
  n8 = Add(A=@n5, B=TextureSample(UVs=Add(A=@n6, B=@n7)) {texture=&T_LightFrost}.B)
  n9 = WorldPosition()
  n10 = VectorParameter() {parameterName=Ice Center}
  n11 = ScalarParameter() {defaultValue=1250, parameterName=Radius}
  n12 = ScalarParameter() {defaultValue=0.2, parameterName=Hardness}
  n13 = SphereMask(A=@n9.XYZ, B=@n10.RGB, Radius=Multiply(A=@n11, B=ScalarParameter() {defaultValue=3, parameterName=Ratio_Hoarfrost}), Hardness=@n12) {hardnessPercent=0}
  n14 = OneMinus(None=SphereMask(A=TextureCoordinate(), B=Constant() {r=0.5}, Radius=LinearInterpolate(Alpha=@n13) {constA=5, constB=0}) {hardnessPercent=0})
  n15 = LinearInterpolate(B=ScalarParameter() {defaultValue=0.06, parameterName=Scalar}, Alpha=ScalarParameter() {parameterName=CountdownStarted}) {constA=1}
  n16 = Multiply(A=@n11, B=ScalarParameter() {defaultValue=3, parameterName=Ratio_Opacity})
  n17 = Clamp(None=LinearInterpolate(Alpha=SphereMask(A=@n9.XYZ, B=@n10.RGB, Radius=@n16, Hardness=@n12) {hardnessPercent=20}))
  n18 = Multiply(A=@n11, B=ScalarParameter() {defaultValue=0.78, parameterName=Ratio_Whiteout})
  n19 = LinearInterpolate(
   A=LinearInterpolate(Alpha=SphereMask(A=@n9.XYZ, B=@n10.RGB, Radius=@n18, Hardness=@n12) {hardnessPercent=20})
   Alpha=ScalarParameter() {parameterName=CountdownStarted}
  ) {constB=0}
  Add(A=Multiply(A=Multiply(A=@n8, B=@n14), B=Multiply(A=@n15, B=@n17)), B=@n19)
  n20 = MaterialFunctionCall() {materialFunction=/Game/Athena/WinterArt/Materials/MF_SharedWinter}
  n21 = Constant() {r=0.5}
  n22 = Multiply(A=SphereMask(A=TextureCoordinate(), B=@n21, Radius=@n21) {hardnessPercent=0}, B=@n5)
  MaterialFunctionCall(Emissive=Add(A=LinearInterpolate(A=@n20.IceColor, B=@n20.IceColor, Alpha=@n22), B=Multiply(A=@n20.SnowColor, B=@n19))) {materialFunction=&MF_FX_EmissiveAdjustments}
 ]
}
