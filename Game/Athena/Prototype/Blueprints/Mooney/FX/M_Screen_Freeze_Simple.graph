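# Unlit translucent sprite material for Niagara and particles that draws a screen-freeze frost overlay with depth testing disabled.
# Three differently tiled samples of a frost texture, one per RGB channel, sum into a pattern masked toward the edges by an inverted sphere mask with a dynamic CenterFallOff exponent.
# That pattern tints shared winter snow color, added to ice color and multiplied by particle color for emissive.
# Particle alpha times the pattern gives opacity.
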Material /Game/Athena/Prototype/Blueprints/Mooney/FX/M_Screen_Freeze_Simple
graph {
 properties {
  bDisableDepthTest=true
  bEnableResponsiveAA=true
  bTangentSpaceNormal=false
  bUsedWithNiagaraSprites=true
  bUsedWithParticleSprites=true
  blendMode=BLEND_Translucent
  ditherOpacityMask=true
  pixelDepthOffsetMode=PDOM_Legacy
  previewMesh=/Engine/Transient.MaterialEditorMeshComponent_7
  shadingModel=MSM_Unlit
  usedShadingModels=Unlit
 }
 nodes [
  n1 = ParticleColor()
  n2 = MaterialFunctionCall() {materialFunction=/Game/Athena/WinterArt/Materials/MF_SharedWinter}
  n3 = TextureCoordinate()
  n4 = Constant() {r=0.5}
  n5 = TextureCoordinate()
  n6 = AppendVector(A=ScalarParameter() {defaultValue=6, parameterName=Frost_U_Tile}, B=ScalarParameter() {defaultValue=6, parameterName=Frost_V_Tile})
  n7 = TextureObjectParameter() {parameterName=Texture, texture=/Game/Effects/Fort_Effects/Textures/Snowflake/T_LightFrost}
  n8 = Multiply(
   A=TextureCoordinate()
   B=AppendVector(A=ScalarParameter() {defaultValue=10, parameterName=Frost_U_Tile2}, B=ScalarParameter() {defaultValue=10, parameterName=Frost_V_Tile2})
  )
  n9 = AppendVector(A=ScalarParameter() {defaultValue=0.06, parameterName=Frost_U_Offset}, B=ScalarParameter() {defaultValue=0.05, parameterName=Frost_V_Offset})
  n10 = Add(
   A=TextureSample(UVs=Multiply(A=@n5, B=@n6), Tex=@n7) {texture=&T_LightFrost}.R
   B=TextureSample(UVs=Add(A=@n8, B=@n9), Tex=@n7) {texture=&T_LightFrost}.G
  )
  n11 = LinearInterpolate(A=@n2.IceColor, B=@n2.IceColor, Alpha=Multiply(A=SphereMask(A=@n3, B=@n4, Radius=@n4) {hardnessPercent=0}, B=@n10))
  n12 = ScalarParameter() {defaultValue=25, parameterName=Frost_U_Tile3}
  n13 = ScalarParameter() {defaultValue=25, parameterName=Frost_V_Tile3}
  n14 = Add(
   A=Multiply(A=TextureCoordinate(), B=AppendVector(A=@n12, B=@n13))
   B=AppendVector(A=ScalarParameter() {defaultValue=0.3, parameterName=Frost_U_Offset2, sliderMin=0.06}, B=ScalarParameter() {parameterName=Frost_V_Offset2})
  )
  n15 = SphereMask(A=TextureCoordinate(), B=Constant() {r=0.5}) {attenuationRadius=0.5, hardnessPercent=0}
  n16 = Multiply(
   A=Add(A=@n10, B=TextureSample(UVs=@n14, Tex=@n7) {texture=&T_LightFrost}.B)
   B=Power(Base=OneMinus(None=@n15), Exp=DynamicParameter() {paramNames [CenterFallOff, Param2, Param3, Param4]}.CenterFallOff) {constExponent=4}
  )
  n17 = Multiply(A=@n2.SnowColor, B=@n16)
  MaterialFunctionCall(Emissive=Multiply(A=@n1.RGB, B=Add(A=@n11, B=@n17))) {materialFunction=&MF_FX_EmissiveAdjustments}
  Multiply(A=@n1.A, B=@n16)
 ]
}
